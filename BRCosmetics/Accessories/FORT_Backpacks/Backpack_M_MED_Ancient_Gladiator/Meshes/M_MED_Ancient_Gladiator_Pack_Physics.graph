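# Physics asset for the Ancient Gladiator backpack cosmetic, covering the pack and its left and right dynamic weapon bones.
# Each of the three bones has one capsule body: a short one on the pack and a long, thin one on each weapon.
# Two constraints join the weapons to the pack, with a small swing-2 limit, slightly wider on the left, and zero swing-1 and twist limits.
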PhysicsAsset /BRCosmetics/Accessories/FORT_Backpacks/Backpack_M_MED_Ancient_Gladiator/Meshes/M_MED_Ancient_Gladiator_Pack_Physics
bodies {
 dyn_Weapons_l [{shapeType=Capsule, length=36.8278, radius=3.1715, center=<1.692 0.2187 2.3932>, rotation=<rot 1.5021 -90 35.0038>}]
 dyn_Weapons_r [{shapeType=Capsule, length=36.8386, radius=3.1689, center=<-1.988 0.0113 2.2223>, rotation=<rot 3.2742 -90 -33.6758>}]
 pack [{shapeType=Capsule, length=3.362, radius=4.3237, center=<-0.0739 -9.0166 -5.4228>, rotation=<rot 0.3657 -90 0.9634>}]
}
constraints [
 {bone1Name=dyn_Weapons_l, bone2Name=pack, swing1LimitDegrees=0, swing2LimitDegrees=12, twistLimitDegrees=0}
 {bone1Name=dyn_Weapons_r, bone2Name=pack, swing1LimitDegrees=0, swing2LimitDegrees=10, twistLimitDegrees=0}
]
